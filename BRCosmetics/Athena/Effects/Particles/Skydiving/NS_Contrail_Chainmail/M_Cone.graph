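# Unlit material for Niagara mesh particles, apparently coloring a cone mesh in the Chainmail skydiving contrail effect.
# The sides blend between two color parameters along the V texture coordinate, using a smoothstep with adjustable min and max.
# Faces whose instance-space normal points nearly straight down get a separate bottom color, avoiding any mask texture.
# The chosen color passes through the shared MF_FX_EmissiveAdjustments function.
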Material /BRCosmetics/Athena/Effects/Particles/Skydiving/NS_Contrail_Chainmail/M_Cone
graph {
 properties {bUsedWithNiagaraMeshParticles=true, pixelDepthOffsetMode=PDOM_Legacy, shadingModel=MSM_Unlit, usedShadingModels=Unlit}
 nodes [
  n1 = Transform(None=VertexNormalWS()) {transformSourceType=TRANSFORMSOURCE_World, transformType=TRANSFORM_Instance}
  n2 = Constant3Vector() {constant=<rgba 0 0 -1 0>}
  n3 = VectorParameter() {parameterName=Color 1}
  n4 = VectorParameter() {parameterName=Color 2}
  n5 = SmoothStep(
   Min=ScalarParameter() {parameterName=Smooth Step Min, sliderMax=1}
   Max=ScalarParameter() {defaultValue=1, parameterName=Smooth Step Max, sliderMax=1}
   Value=ComponentMask(None=TextureCoordinate()) {g=true}
  )
  MaterialFunctionCall(Emissive=If(
    A=DotProduct(A=@n1, B=@n2)
    "A > B"=VectorParameter() {parameterName=Color Bottom, defaultValue=<rgba 0 0 0 1>}.RGB
    "A < B"=LinearInterpolate(A=@n3.RGB, B=@n4.RGB, Alpha=@n5)
   ) {constB=0.9}) {materialFunction=&MF_FX_EmissiveAdjustments}
 ]
}
