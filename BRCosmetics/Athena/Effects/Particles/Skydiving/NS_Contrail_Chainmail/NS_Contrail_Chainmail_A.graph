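# Niagara skydiving contrail cosmetic "Chainmail": mesh chains trailing blue-purple flames, cones, embers and glowing lights from sampled character sockets.
# Skydive velocity aims the trails and Dive scales spawn rates, size and brightness, while FrontEnd modifiers retune values for menu display.
# Flame ribbons use separate frontend and gameplay emitters, the latter bending local-space velocity toward smoothed skydive velocity.
# Flame sprites re-snap to sockets, then offset along velocity by an age curve.
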NiagaraSystem /BRCosmetics/Athena/Effects/Particles/Skydiving/NS_Contrail_Chainmail/NS_Contrail_Chainmail_A
dependencies {
 usedDataInterfaces [
  &NiagaraDataInterfaceCurve
  &NiagaraDataInterfaceArrayFloat3
  &NiagaraDataInterfaceArrayFloat
  &NiagaraDataInterfaceColorCurve
 ]
 usedDynamicInputs [&SetBoolByFloatComparison, &Subtract_Float]
 usedModules [
  &self:SystemSpawnScript.SystemScriptSource.SystemScriptGraph.NiagaraNodeAssignment_3.SetVariables_#1
  &SystemState
  /Game/Effects/Niagara/Modules/Skydiving/FN_DivingAttributes/v4/FN_DivingAttributes_v4
  &self:SystemSpawnScript.SystemScriptSource.SystemScriptGraph.NiagaraNodeAssignment_0.SetVariables_#2
  &self:SystemSpawnScript.SystemScriptSource.SystemScriptGraph.NiagaraNodeAssignment_1.SetVariables_#3
  &self:SystemSpawnScript.SystemScriptSource.SystemScriptGraph.NiagaraNodeAssignment_2.SetVariables_#4
  &self:SystemSpawnScript.SystemScriptSource.SystemScriptGraph.NiagaraNodeAssignment_5.SetVariables_#5
  &self:SystemSpawnScript.SystemScriptSource.SystemScriptGraph.NiagaraNodeAssignment_4.SetVariables_#6
  &Timeline
 ]
}
emitters {
 Chains {
  emitterSpawnScript [&FN_PreserveSkydiveUserVariables() {module=PreserveUserVariables}]
  emitterUpdateScript [
   &EmitterState(Inactive Response=::NewEnumerator1, Life Cycle Mode=::NewEnumerator1, Loop Duration=0, UseLoopDelay=false)
   &SpawnBurst_Instantaneous(Spawn Count=2, Spawn Probability=HlslExpression() {hlslExpression="Emitter.LoopCount == 0"}, Use Spawn Probability=true) {module=SpawnBurst_Instantaneous001}
  ]
  particleSpawnScript [
   &FN_SetSpawnGroup(Number of Spawn Groups=2)
   &InitializeParticle(
    Color Mode=::NewEnumerator0
    Lifetime=0
    Mesh Scale Mode=::NewEnumerator3
    Sprite Size Mode=::NewEnumerator3
    Uniform Sprite Size=20
    UsePositionOffset=false
    Write Material Random=true
   ) {module=InitializeParticle001}
   &Color(Color=$NPC.FortniteNPC.FortniteActiveTimeOfDay)
   &DynamicMaterialParameters(
    Index 0 Param 1=&RandomRangeFloat(Minimum=0, Maximum=3.333)
    Index 0 Param 2=&FN_FrontEndFloatModifier(Gameplay=5, Front End=4)
    Index 0 Param 3=&FN_FrontEndFloatModifier(Gameplay=0, Front End=15)
    Index 0 Param 4=&FN_FrontEndFloatModifier(Gameplay=55, Front End=22)
    Module.Param0WriteEnabled=true
    Module.Param1WriteEnabled=true
    Module.Param2WriteEnabled=true
    Module.Param3WriteEnabled=true
   )
   /SetVariables(
    Particles.MeshIndex=/Niagara/DynamicInputs/Divide/Divide_Int(A:NiagaraInt32=&ReturnParticleIDAttributes(), B:NiagaraInt32=4)
    Particles.Scale=&FN_FrontEndVector3Modifier(
     Gameplay=&VectorFromFloat(Value=1.5)
     Front End=&VectorFromFloat(Value=&Lerp_Float(
       A=1.3
       B=0.8
       Alpha=&MakeFloatFromInt(Int=&Divide_Int(A:NiagaraInt32=&SpawnGroup(), B:NiagaraInt32=2))
      ))
    )
   ) {isSetParameters=true}
   &InitialMeshOrientation(
    Mesh Orientation Mode=::NewEnumerator7
    Rotation=&MakeVector(X=&RandomRangeFloat(Minimum=0, Maximum=1), Y=0, Z=0)
    Rotation Coordinate Space=::NewEnumerator0
    Use Rotation Vector=true
   )
  ]
  particleUpdateScript [
   &ParticleState(Kill Particles When Lifetime Has Elapsed=false)
   &ScaleMeshSize(Scale Factor=&VectorFromFloat(Value=&FloatFromCurve(FloatCurve=DataInterface() {propertyValues {Curve {keys [<at 0.1 1>,<at 1 1.5>]}}}, CurveIndex=$System.Dive)))
   &SkeletalMeshLocation(
    Bone / Socket Index=$Particles.SpawnGroup
    Bone Sampling Mode=::NewEnumerator4
    Inherit Velocity=false
    Interpolate Mesh Sampling=false
    MaskSpawnGroup=false
    Sampled Position Offset=&SelectVectorFromArray(
     Direct Array Index=&SpawnGroup()
     Vector Selection Array=DataInterface() {propertyValues {FloatData [<0 0 0>, <0 1 1>, <7.5 0 0>, <-7.5 0 0>]}}
    )
    Sampled Position Offset Coordinate Space:ENiagaraMeshTransforms=::NewEnumerator3
    Skeletal Mesh=$User.SkeletalMesh
   )
   /Niagara/Modules/Update/Orientation/V2/OrientMeshToVector(
    Look At Direction:Vector3f=&NormalizeVector(
     Vector To Normalize=&FN_FrontEndVector3Modifier(
      Gameplay=&Multiply_VectorByFloat(
       Vector=&TransformVector(Vector=$Engine.Owner.Velocity, Source Space=::World, Destination Space=::Simulation)
       Float=-1
      )
      Front End=&TransformVector(
       Vector=&NormalizeVector(
        Vector To Normalize=&Lerp_Vector(A=<0 -0.2 2>, B=<0 -2 2>, Alpha=$User.ParaGlideLeanAlpha)
        Safe Normalize=true
        Safe Fallback Threshold=0
        Safe Fallback Vector=<0 0 0.001>
       )
       Destination Space=::Simulation
      )
     )
     Safe Normalize=false
     Safe Fallback Threshold=0
     Safe Fallback Vector=<0 0 0.001>
    )
    Override Quat=false
    Side Axis Direction:Vector3f=&TransformVector(Vector=<0 1 0>, Destination Space=::World)
    UseOptionalSideAxis=false
    UseOptionalUpAxis=false
   )
   &UpdateMeshOrientation(
    Rotation Coordinate Space=::NewEnumerator3
    Rotation Rate=&RandomRangeFloat(Minimum=-1, Maximum=1)
    Rotation Vector=<1 0 0>
   )
   /SetVariables(Particles.LightColor=&MultiplyLinearColor(
     Linear Color=<rgba 0.1667 0 5 1>
     Scale Factor=&MakeVector4_FromFloat(Float=&FloatFromCurve(FloatCurve=DataInterface() {propertyValues {Curve {keys [<at 0 1>,<at 1 2>]}}}, CurveIndex=$System.Dive))
    )) {isSetParameters=true, module=SetVariables_#7}
   /SetVariables(Particles.SpriteSize=&Multiply_Vector2D_ByFloat(Float=&Lerp_Float(A=1, B=2, Alpha=$System.Dive), Vector2D=$Particles.Initial.SpriteSize)) {isSetParameters=true, module=SetVariables_#8}
  ]
  rendererClasses [&NiagaraMeshRendererProperties]
  renderers [{rendererClass=&NiagaraMeshRendererProperties, rendererIndex=0}]
 }
 Cones {
  emitterUpdateScript [
   &EmitterState(Life Cycle Mode=::NewEnumerator1, Loop Behavior=::NewEnumerator1, Loop Duration Mode=::NewEnumerator1, UseLoopDelay=false)
   &SpawnRate(
    Spawn Probability=0.8
    SpawnRate=&Lerp_Float(
     A=35
     B=40
     Alpha=&FloatFromCurve(FloatCurve=DataInterface() {propertyValues {Curve {keys [<at 0 0>,<at 0.3 1>]}}}, CurveIndex=$System.Dive)
    )
    Use Spawn Probability=false
   )
  ]
  particleSpawnScript [
   &FN_SetSpawnGroup()
   &InitializeParticle(
    Color=<rgba 0.0408 0.0312 0.0833 1>
    Lifetime=&Lerp_Float(
     A=0.45
     B=0.25
     Alpha=&MakeFloatFromInt(Int=&Divide_Int(A:NiagaraInt32=&SpawnGroup(), B:NiagaraInt32=2))
    )
    Mesh Scale=&Multiply_VectorByFloat(Vector=<1 1 3>, Float=&RandomRangeFloat(Minimum=0.02, Maximum=0.03))
    Mesh Scale Mode=::NewEnumerator4
    UsePositionOffset=false
    Write Material Random=true
   )
   &InitialMeshOrientation(Mesh Orientation Mode=::NewEnumerator7, Rotation=<0 0.25 0>, Use Rotation Vector=true)
   &SkeletalMeshLocation(
    Bone / Socket Index=&SpawnGroup()
    Bone Sampling Mode=::NewEnumerator4
    Interpolate Mesh Sampling=false
    MaskSpawnGroup=false
    Orientation Sampling=::NewEnumerator1
    Sampled Position Offset Coordinate Space:ENiagaraSkelMeshTransforms=::NewEnumerator2
    Skeletal Mesh=$User.SkeletalMesh
   )
   &ShapeLocation(MaskSpawnGroup=false, Override Seed=false, Rotation Mode=::NewEnumerator5, Sphere Radius=5, UseOwnerScale=true)
   &AddVelocity(
    Cone Angle=40
    Cone Axis=&Multiply_VectorByFloat(
     Vector=&NormalizeVector(
      Vector To Normalize=&TransformVector(Vector=$System.SkydiveVelocity, Source Space=::World, Destination Space=::Local)
      Safe Normalize=true
      Safe Fallback Threshold=0
      Safe Fallback Vector=<0 0 0.001>
     )
     Float=-1
    )
    Module.UseConeFalloff=false
    Module.UseVelDistribution=false
    Override Seed=false
    Rotation Mode=::NewEnumerator5
    Velocity Mode=::NewEnumerator2
    Velocity Speed=&Multiply_Float(
     A=&FloatFromCurve(
      FloatCurve=DataInterface() {propertyValues {Curve {keys [{interpMode=RCIM_Cubic, tangentMode=RCTM_User, time=0, value=1}, {interpMode=RCIM_Cubic, tangentMode=RCTM_User, time=1, value=2}]}}}
      CurveIndex=$System.Dive
     )
     B=&Multiply_Float(
      A=&RandomRangeFloat(Minimum=600, Maximum=700)
      B=&Lerp_Float(
       A=0.8
       B=0.5
       Alpha=&MakeFloatFromInt(Int=&Divide_Int(A:NiagaraInt32=&SpawnGroup(), B:NiagaraInt32=2))
      )
     )
    )
   )
  ]
  particleUpdateScript [
   &ParticleState()
   &ScaleColor(
    Scale Alpha=&FloatFromCurve(FloatCurve=DataInterface() {propertyValues {Curve {keys [
         <at 0.5 1 RCIM_Cubic>
         {arriveTangent=-0.0575, interpMode=RCIM_Cubic, leaveTangent=-0.0575, tangentMode=RCTM_User, time=0.8, value=1}
         {arriveTangent=0.9752, interpMode=RCIM_Cubic, leaveTangent=0.9752, tangentMode=RCTM_User, time=1, value=0}
        ]}}})
    ScaleA=true
    ScaleRGB=false
   )
   &ScaleMeshSize(Scale Factor=&VectorFromFloat(Value=&FloatFromCurve(FloatCurve=DataInterface() {propertyValues {Curve {keys [
          {tangentMode=RCTM_User, time=0, value=1}
          {arriveTangent=-0.0405, interpMode=RCIM_Cubic, leaveTangent=-0.0405, tangentMode=RCTM_User, time=0.6, value=1}
          {arriveTangent=-3.3906, interpMode=RCIM_Cubic, leaveTangent=-3.3906, tangentMode=RCTM_User, time=1, value=0}
         ]}}})))
   &CurlNoiseForce(Noise Strength=250, Pan Noise=false)
   &Drag(Drag=&RandomRangeFloat(Minimum=5, Maximum=6), Use Linear Drag=true, Use Rotational Drag=false)
   &SolveForcesAndVelocity(Clamp Velocity=false, Limit Acceleration=false)
   &CameraOffset(Camera Offset Amount=20) {enabled=false}
  ]
  rendererClasses [&NiagaraMeshRendererProperties]
  renderers [{rendererClass=&NiagaraMeshRendererProperties, rendererIndex=0}]
 }
 Embers {
  emitterUpdateScript [
   &EmitterState()
   &SpawnRate(SpawnRate=&FN_FrontEndFloatModifier(Gameplay=150, Front End=200), Use Spawn Probability=false)
  ]
  particleSpawnScript [
   &FN_SetSpawnGroup()
   /SetVariables(Particles.PerParticleRandom:NiagaraFloat=&RandomRangeFloat(Minimum=0, Maximum=1)) {isSetParameters=true}
   &InitializeParticle(
    Color Mode=::NewEnumerator0
    Lifetime=&Lerp_Float(A=0.5, B=0.75, Alpha=$Particles.PerParticleRandom)
    Position Mode=::NewEnumerator0
    Sprite Size=&Multiply_Vector2D(
     A=&RandomRangeVector2D(Minimum=<0.5 1>, Maximum=<0.5 2>)
     B=&MakeVector2D(
      X=&FloatFromCurve(
       FloatCurve=DataInterface() {propertyValues {Curve {keys [<at 0 1>,<at 0.8 2 RCIM_Constant>,<at 0.95 3 RCIM_Constant>]}}}
       CurveIndex=$Particles.PerParticleRandom
       Scale Curve=0.75
      )
      Y=&FloatFromCurve(
       FloatCurve=DataInterface() {propertyValues {Curve {keys [<at 0 1>,<at 0.8 2 RCIM_Constant>,<at 0.95 3>]}}}
       CurveIndex=$Particles.PerParticleRandom
      )
     )
    )
    Sprite Size Mode=::NewEnumerator4
    Write Material Random=true
   )
   &SkeletalMeshLocation(
    Bone / Socket Index=$Particles.SpawnGroup
    Bone Sampling Mode=::NewEnumerator4
    Inherit Velocity=false
    Interpolate Mesh Sampling=false
    MaskSpawnGroup=false
    Sampled Position Offset=&SelectVectorFromArray(
     Direct Array Index=&SpawnGroup()
     Vector Selection Array=DataInterface() {propertyValues {FloatData [<0 -5 0>, <0 5 0>, <0 0 0>, <0 0 0>]}}
    )
    Sampled Position Offset Coordinate Space:ENiagaraMeshTransforms=::NewEnumerator3
    Skeletal Mesh=$User.SkeletalMeshFrontend
   )
   &ShapeLocation(MaskSpawnGroup=false, Override Seed=false, Sphere Radius=5, UseOwnerScale=true)
   &AddVelocity(
    Cone Angle=33
    Cone Axis=&Multiply_VectorByFloat(
     Vector=&NormalizeVector(
      Vector To Normalize=&TransformVector(Vector=$System.SkydiveVelocity, Source Space=::World, Destination Space=::Local)
      Safe Normalize=true
      Safe Fallback Threshold=0
      Safe Fallback Vector=<0 0 0.001>
     )
     Float=-1
    )
    Module.UseConeFalloff=false
    Module.UseVelDistribution=false
    Override Seed=false
    Velocity Mode=::NewEnumerator2
    Velocity Speed=&Multiply_Float(
     A=&Lerp_Float(A=400, B=500, Alpha=$Particles.PerParticleRandom)
     B=&Lerp_Float(
      A=0.7
      B=1
      Alpha=&MakeFloatFromInt(Int=&Divide_Int(A:NiagaraInt32=&SpawnGroup(), B:NiagaraInt32=2))
     )
    )
   )
  ]
  particleUpdateScript [
   &ParticleState()
   &AccelerationForce()
   &Drag(
    Drag=&FloatFromCurve(
     FloatCurve=DataInterface() {propertyValues {Curve {keys [<at 0 0.5>,<at 0.5 0.5>,<at 1 1>]}}}
     Scale Curve=&RandomRangeFloat(Minimum=6, Maximum=22)
    )
    Use Linear Drag=true
    Use Rotational Drag=false
   )
   /Niagara/Modules/Update/Forces/CurlNoiseForce(
    Noise Frequency=0.04
    Noise Strength=&FloatFromCurve(
     FloatCurve=DataInterface() {propertyValues {Curve {keys [{tangentMode=RCTM_User, time=0, value=0}, {tangentMode=RCTM_User, time=1, value=1}]}}}
     Scale Curve=&Lerp_Float(A=4444, B=444, Alpha=$Particles.PerParticleRandom)
    )
    Pan Noise=true
    Pan Noise Field=<0 0 1>
   )
   &SolveForcesAndVelocity(Clamp Velocity=false, Limit Acceleration=false)
   &ScaleColor(
    Scale RGB=&MakeVectorFromLinearColorRGB(Color=&ColorFromCurve(ColorCurve=DataInterface() {propertyValues {
        AlphaCurve {keys [<at 0 1 RCIM_Cubic>]}
        BlueCurve {keys [<at 0 10 RCIM_Cubic>,<at 0.2445 1>,<at 0.473 0.5885 RCIM_Cubic>]}
        GreenCurve {keys [<at 0 3.8906 RCIM_Cubic>,<at 0.2445 0>,<at 0.473 0 RCIM_Cubic>]}
        RedCurve {keys [<at 0 0 RCIM_Cubic>,<at 0.2445 0.8416>,<at 0.473 0.0043 RCIM_Cubic>]}
       }}))
    ScaleA=true
    ScaleRGB=true
   )
   &ScaleSpriteSizeBySpeed(
    Max Scale Factor=<1 2>
    Sample Scale Factor By Curve=true
    Scale Factor Curve=DataInterface() {propertyValues {Curve {keys [<at 0 0>, {arriveTangent=1.4989, interpMode=RCIM_Cubic, tangentMode=RCTM_Break, time=1, value=1}]}}}
    Velocity Threshold=999
   )
   &ScaleSpriteSize(
    Uniform Curve Scale=&MakeFloatFromVector(Vector=$Engine.Owner.Scale, Channel=::NewEnumerator0)
    Uniform Curve Sprite Scale=DataInterface() {propertyValues {Curve {keys [
        {tangentMode=RCTM_User, time=0, value=1}
        {arriveTangent=-0.0405, interpMode=RCIM_Cubic, leaveTangent=-0.0405, tangentMode=RCTM_User, time=0.7009, value=1}
        {arriveTangent=-8.0089, interpMode=RCIM_Cubic, leaveTangent=-8.0089, tangentMode=RCTM_User, time=1, value=0}
       ]}}}
   ) {enabled=false}
  ]
  rendererClasses [&NiagaraSpriteRendererProperties]
  renderers [{rendererClass=&NiagaraSpriteRendererProperties}]
 }
 FlamesSprites {
  emitterUpdateScript [
   &EmitterState(Life Cycle Mode=::NewEnumerator1, Loop Behavior=::NewEnumerator1, Loop Duration Mode=::NewEnumerator1, UseLoopDelay=false)
   &SpawnRate(
    SpawnRate=&Lerp_Float(A=30, B=&FN_FrontEndFloatModifier(Gameplay=60, Front End=100), Alpha=$System.Dive)
    Use Spawn Probability=false
   )
  ]
  particleSpawnScript [
   &FN_SetSpawnGroup()
   /SetVariables(Particles.ParticleRandom:NiagaraFloat=&RandomRangeFloat(Minimum=0, Maximum=1)) {isSetParameters=true}
   &InitializeParticle(
    Color=<rgba 1 1 1 0.7>
    Lifetime=&RandomRangeFloat(Minimum=1.2, Maximum=1.5)
    Sprite Rotation Mode=::NewEnumerator2
    Sprite Size Mode=::NewEnumerator3
    Uniform Sprite Size=&Lerp_Float(
     A=15
     B=20
     Alpha=&MakeFloatFromInt(Int=&Divide_Int(A:NiagaraInt32=&SpawnGroup(), B:NiagaraInt32=2))
    )
    UsePositionOffset=true
    Write Material Random=true
   )
   &SkeletalMeshLocation(
    Bone / Socket Index=&SpawnGroup()
    Bone Sampling Mode=::NewEnumerator4
    MaskSpawnGroup=false
    Orientation Sampling=::NewEnumerator1
    Skeletal Mesh=$User.SkeletalMeshFrontend
   )
   &DynamicMaterialParameters(
    Index 0 Param 2=&UniformRangedFloat(Minimum=0, Maximum=1)
    Index 0 Param 4=&UniformRangedFloat(Minimum=3, Maximum=5)
    Module.Param0WriteEnabled=false
    Module.Param1WriteEnabled=true
    Module.Param2WriteEnabled=false
    Module.Param3WriteEnabled=true
   )
   &AddVelocity(
    Cone Angle=&FN_FrontEndFloatModifier(Gameplay=10, Front End=5)
    Cone Axis=&Multiply_VectorByFloat(
     Vector=&NormalizeVector(
      Vector To Normalize=&TransformVector(Vector=$System.SkydiveVelocity, Source Space=::World, Destination Space=::Local)
      Safe Normalize=true
      Safe Fallback Threshold=0
      Safe Fallback Vector=<0 0 0.001>
     )
     Float=-1
    )
    Module.UseConeFalloff=false
    Module.UseVelDistribution=false
    Override Seed=false
    Velocity Mode=::NewEnumerator2
    Velocity Speed=&FN_FrontEndFloatModifier(
     Gameplay=&Multiply_Float(
      A=&Lerp_Float(A=100, B=120, Alpha=$Particles.ParticleRandom)
      B=&FloatFromCurve(
       FloatCurve=DataInterface() {propertyValues {Curve {keys [<at 0 1 RCIM_Cubic>, {arriveTangent=1.6296, interpMode=RCIM_Cubic, leaveTangent=1.6296, tangentMode=RCTM_User, time=1, value=2.375}]}}}
       CurveIndex=$System.Dive
      )
     )
     Front End=&RandomRangeFloat(Minimum=75, Maximum=100)
    )
   ) {enabled=false, module=AddVelocity001}
  ]
  particleUpdateScript [
   &ParticleState()
   /SetVariables(Particles.SubImageIndex=&FloatFromCurve(
     FloatCurve=DataInterface() {propertyValues {Curve {keys [{tangentMode=RCTM_User, time=0, value=0.12}, {tangentMode=RCTM_User, time=1, value=1}]}}}
     Scale Curve=63
    )) {isSetParameters=true}
   &DynamicMaterialParameters(
    Index 0 Param 1=&FloatFromCurve(FloatCurve=DataInterface() {propertyValues {Curve {keys [<at 0 0>,<at 0.8 1>,<at 1.0011 1>]}}})
    Index 0 Param 2=2.5
    Index 0 Param 3=1
    Index 0 Param 4=2
    Module.Param0WriteEnabled=true
    Module.Param1WriteEnabled=false
    Module.Param2WriteEnabled=true
    Module.Param3WriteEnabled=true
   ) {module=DynamicMaterialParameters001}
   &ScaleSpriteSize(Scale Factor=&Vector2DFromFloat(Value=&FloatFromCurve(
      FloatCurve=DataInterface() {propertyValues {Curve {keys [
          {arriveTangent=0.0497, interpMode=RCIM_Cubic, leaveTangent=0.0497, tangentMode=RCTM_User, time=0, value=1}
          {arriveTangent=-0.103, interpMode=RCIM_Cubic, leaveTangent=-0.103, tangentMode=RCTM_User, time=1, value=1.25}
         ]}}}
      Scale Curve=&FN_FrontEndFloatModifier(Gameplay=2.5, Front End=2)
     )))
   &ScaleColor(
    Scale Alpha=&FloatFromCurve(FloatCurve=DataInterface() {propertyValues {Curve {keys [
         {tangentMode=RCTM_User, time=0, value=1}
         {arriveTangent=-0.0425, interpMode=RCIM_Cubic, leaveTangent=-0.0425, tangentMode=RCTM_User, time=0.3, value=1}
         {arriveTangent=-1.9855, interpMode=RCIM_Cubic, leaveTangent=-1.9855, tangentMode=RCTM_User, time=1, value=0.6}
        ]}}})
    Scale RGB=&MakeVectorFromLinearColorRGB(Color=&ColorFromCurve(ColorCurve=DataInterface() {propertyValues {
        AlphaCurve {keys [<at 0.2926 1 RCIM_Cubic>,<at 0.5529 0.5>]}
        BlueCurve {keys [
          {arriveTangent=2.3844, interpMode=RCIM_Cubic, leaveTangent=2.3844, tangentMode=RCTM_User, time=0.1049, value=5}
          <at 0.2631 3>
          <at 0.4703 0.5 RCIM_Cubic>
         ]}
        GreenCurve {keys [
          {arriveTangent=2.3844, interpMode=RCIM_Cubic, leaveTangent=2.3844, tangentMode=RCTM_User, time=0.1049, value=2.2848}
          <at 0.2631 0>
          <at 0.4703 0.0167 RCIM_Cubic>
         ]}
        RedCurve {keys [
          {arriveTangent=2.3844, interpMode=RCIM_Cubic, leaveTangent=2.3844, tangentMode=RCTM_User, time=0.1049, value=0}
          <at 0.2631 0.8>
          <at 0.4703 0 RCIM_Cubic>
         ]}
       }}))
    ScaleA=true
    ScaleRGB=true
   )
   &SpriteRotationRate(Rotation Rate=&RandomRangeFloat(Minimum=-66, Maximum=66)) {enabled=false}
   &SkeletalMeshLocation(
    Bone / Socket Index=&SpawnGroup()
    Bone Sampling Mode=::NewEnumerator4
    MaskSpawnGroup=false
    Orientation Sampling=::NewEnumerator1
    Sampled Position Offset=&FN_FrontEndVector3Modifier(
     Gameplay=&SelectVectorFromArray(
      Direct Array Index=&SpawnGroup()
      Vector Selection Array=DataInterface() {propertyValues {FloatData [<0 -7 0>, <0 7 0>, <0 0 0>, <0 0 0>]}}
     )
     Front End=&SelectVectorFromArray(
      Direct Array Index=&SpawnGroup()
      Vector Selection Array=DataInterface() {propertyValues {FloatData [<0 -5 0>, <0 5 0>, <0 0 0>, <0 0 0>]}}
     )
    )
    Skeletal Mesh=$User.SkeletalMeshFrontend
    Use Sampled Mesh Attribute as Ribbon ID=false
   ) {module=SkeletalMeshLocation001}
   /SetVariables(Particles.PositionControl:NiagaraFloat=&FloatFromCurve(
     FloatCurve=DataInterface() {propertyValues {Curve {keys [
         {arriveTangent=0.4185, interpMode=RCIM_Cubic, leaveTangent=0.4185, tangentMode=RCTM_User, time=0, value=0}
         {arriveTangent=2.0966, interpMode=RCIM_Cubic, leaveTangent=2.0966, tangentMode=RCTM_User, time=1, value=1}
        ]}}}
     Scale Curve=&Multiply_Float(
      A=&Multiply_Float(
       A=&RandomRangeFloat(Minimum=-100, Maximum=-150)
       B=&Lerp_Float(
        A=0.4
        B=1
        Alpha=&MakeFloatFromInt(Int=&Divide_Int(A:NiagaraInt32=&SpawnGroup(), B:NiagaraInt32=2))
       )
      )
      B=&FloatFromCurve(
       FloatCurve=DataInterface() {propertyValues {Curve {keys [{arriveTangent=1.6154, time=0, value=1}, {arriveTangent=2.6178, leaveTangent=2.6178, time=1, value=1.55}]}}}
       CurveIndex=$System.Dive
      )
     )
    )) {isSetParameters=true}
   /SetVariables(Particles.Position=&AddVectorToPosition(
     Position=$Particles.Position
     Vector=&Multiply_VectorByFloat(
      Vector=&Add_Vector(
       A=&NormalizeVector(
        Vector To Normalize=&TransformVector(Vector=$System.SkydiveVelocity, Source Space=::World, Destination Space=::Simulation)
        Safe Normalize=true
        Safe Fallback Threshold=0
        Safe Fallback Vector=<0 0 0.001>
       )
       B=&RandomVector(Vector Scale=&Lerp_Float(A=0, B=&FN_FrontEndFloatModifier(Gameplay=0.002, Front End=0.02), Alpha=$System.Dive))
      )
      Float=$Particles.PositionControl
     )
    )) {isSetParameters=true}
   &InterpolateOverTime(
    Rate Of Change:NiagaraFloat=&FloatFromCurve(
     FloatCurve=DataInterface() {propertyValues {Curve {keys [
         <at 0 10>
         {arriveTangent=-7.727, interpMode=RCIM_Cubic, leaveTangent=-7.727, tangentMode=RCTM_User, time=0.01, value=4}
         <at 1 1 RCIM_Cubic>
        ]}}}
     Scale Curve=&Lerp_Float(
      A=10
      B=15
      Alpha=&RemapRange(
       Input Value To Remap=&VectorLength(Vector=$System.SkydiveVelocity)
       Input Min=1200
       Input Max=2000
       Output Min=0
       Output Max=1
      )
     )
    )
    Target Vector Value:NiagaraPosition=$Particles.Position
    Variable Type=::NewEnumerator2
   ) {enabled=false}
   &Drag(Drag=3, Use Linear Drag=true, Use Rotational Drag=true) {enabled=false}
   &SolveForcesAndVelocity(Clamp Velocity=false, Limit Acceleration=false) {enabled=false}
   /SetVariables(Particles.Position=$StackContext.InterpolateOverTime.InterpolateOverTime_V3_.Moving Average) {enabled=false, isSetParameters=true}
   &CameraOffset(Camera Offset Amount=5)
  ]
  rendererClasses [&NiagaraSpriteRendererProperties]
  renderers [{rendererClass=&NiagaraSpriteRendererProperties}]
 }
 Flames_Frontend {
  emitterUpdateScript [
   &EmitterState()
   &FN_SpawnRibbon_v2(
    Number of Ribbons=4
    Spawn Mode=::NewEnumerator0
    SpawnRate:NiagaraFloat=&FN_FrontEndFloatModifier(Gameplay=0, Front End=100)
   )
   &FN_Integrator()
  ]
  particleSpawnScript [
   &FN_SetSpawnGroup() {module=FN_SetSpawnGroup001}
   /Game/Effects/Niagara/Modules/Ribbons/FN_InitializeRibbon_v4(
    Custom Offset=true
    Enable Skeletal Position=true
    Lifetime=1
    Local Vector Offset:Vector3f=&Add_Vector(
     A=&Multiply_VectorByFloat(
      Vector=&NormalizeVector(
       Vector To Normalize=&TransformVector(Vector=$System.SkydiveVelocity, Source Space=::World, Destination Space=::Local)
       Safe Normalize=true
       Safe Fallback Threshold=0
       Safe Fallback Vector=<0 0 0.001>
      )
      Float=&Lerp_Float(
       A=25
       B=20
       Alpha=&MakeFloatFromInt(Int=&Divide_Int(A:NiagaraInt32=&SpawnGroup(), B:NiagaraInt32=2))
      )
     )
     B=<0 0 0>
    )
    Offset Amount:NiagaraFloat=&Lerp_Float(A=1, B=1, Alpha=$System.Dive)
    Ribbon Width=25
    SkeletalMesh:NiagaraDataInterfaceSkeletalMesh=$User.SkeletalMeshFrontend
   )
   &FN_FlipRibbonV() {enabled=false}
   &AddVelocity(Velocity=&FN_FrontEndVector3Modifier(
     Gameplay=<0 0 0>
     Front End=&Multiply_VectorByFloat(
      Vector=&NormalizeVector(
       Vector To Normalize=&TransformVector(Vector=$System.SkydiveVelocity, Source Space=::World, Destination Space=::Local)
       Safe Normalize=true
       Safe Fallback Threshold=0
       Safe Fallback Vector=<0 0 0.001>
      )
      Float=-1200
     )
    ))
   /SetVariables(Particles.Lifetime=&Sine(
     Normalized Angle=&Add_Float(
      A=$Emitter.Age
      B=&SelectFloatFromArray(
       Direct Array Index=&SpawnGroup()
       Float Selection Array=DataInterface() {propertyValues {FloatData [0,0.25,0.5,0.75]}}
      )
     )
     Period=&SelectFloatFromArray(
      Direct Array Index=&SpawnGroup()
      Float Selection Array=DataInterface() {propertyValues {FloatData [1.8,1,1.5,2]}}
     )
     Scale=&Lerp_Float(A=0.01, B=0.005, Alpha=$System.Dive)
     Bias=&Multiply_Float(
      A=&Lerp_Float(A=0.15, B=0.2, Alpha=$System.Dive)
      B=&Lerp_Float(
       A=0.6
       B=1
       Alpha=&MakeFloatFromInt(Int=&Divide_Int(A:NiagaraInt32=&SpawnGroup(), B:NiagaraInt32=2))
      )
     )
    )) {isSetParameters=true, module=SetVariables_#9}
   /SetVariables(Particles.RibbonWidth=&Lerp_Float(
     A=30
     B=25
     Alpha=&MakeFloatFromInt(Int=&Divide_Int(A:NiagaraInt32=&SpawnGroup(), B:NiagaraInt32=2))
    )) {isSetParameters=true}
   &SkeletalMeshLocation(
    Bone / Socket Index=$Particles.SpawnGroup
    Bone Sampling Mode=::NewEnumerator4
    Interpolate Mesh Sampling=false
    MaskSpawnGroup=false
    Orientation Sampling=::NewEnumerator1
    Position Sampling=::NewEnumerator2
    Sampled Position Offset=&SelectVectorFromArray(
     Direct Array Index=&SpawnGroup()
     Vector Selection Array=DataInterface() {propertyValues {FloatData [<0 0 0>, <0 0 0>, <0 0 0>, <0 0 0>]}}
    )
    Skeletal Mesh=$User.SkeletalMesh
   )
   /SetVariables(Particles.Position=&AddVectorToPosition(
     Position=$Particles.Position
     Vector=&RotateVectorByQuaternion(
      Vector To Rotate=&SelectVectorFromArray(
       Direct Array Index=&SpawnGroup()
       Vector Selection Array=DataInterface() {propertyValues {FloatData [<0 -5 0>, <0 5 0>, <0 0 0>, <0 0 0>]}}
      )
      Quaternion=$Particles.SkeletalMeshLocation.SampledRotation
     )
    )) {isSetParameters=true}
  ]
  particleUpdateScript [
   &SolveForcesAndVelocity(Clamp Velocity=false, Limit Acceleration=false)
   &DynamicMaterialParameters(
    Index 0 Param 1=&self:NormalizedRibbonID(Number of Ribbons:NiagaraInt32=4, Ribbon ID:NiagaraID=$Particles.RibbonID)
    Index 0 Param 2=$System.Dive
    Index 0 Param 3=-15
    Module.Param0WriteEnabled=true
    Module.Param1WriteEnabled=true
    Module.Param2WriteEnabled=true
    Module.Param3WriteEnabled=false
   )
   /Game/Effects/Niagara/Modules/Ribbons/FN_UpdateRibbon_v4(
    Detach Locked Particle=true
    Enable Skeletal Position=true
    Locked Particle Detach Time=0
    SkeletalMesh:NiagaraDataInterfaceSkeletalMesh=$User.SkeletalMeshFrontend
   )
   &FN_RibbonMaterial() {enabled=false}
   &ScaleRibbonWidth(Ribbon Width Scale=&FloatFromCurve(FloatCurve=DataInterface() {propertyValues {Curve {keys [
         {arriveTangent=-2.5695, interpMode=RCIM_Cubic, leaveTangent=-2.5695, tangentMode=RCTM_User, time=0, value=1.55}
         {arriveTangent=-2.8721, interpMode=RCIM_Cubic, leaveTangent=-2.8721, tangentMode=RCTM_User, time=1, value=0}
        ]}}}))
  ]
  rendererClasses [&NiagaraRibbonRendererProperties]
  renderers [{rendererClass=&NiagaraRibbonRendererProperties}]
 }
 Flames_Gameplay_LocalSpace {
  emitterUpdateScript [
   &EmitterState()
   &FN_SpawnRibbon_v2(
    Number of Ribbons=4
    Spawn Mode=::NewEnumerator0
    SpawnRate:NiagaraFloat=&Multiply_Float(
     A=&Lerp_Float(
      A=128
      B=256
      Alpha=&RemapRange(
       Input Value To Remap=&VectorLength(Vector=$System.SkydiveVelocity)
       Input Min=1200
       Input Max=2000
       Output Min=0
       Output Max=1
      )
     )
     B=&FN_FrontEndFloatModifier(Gameplay=1, Front End=0)
    )
   )
   &FN_Integrator()
  ]
  particleSpawnScript [
   &FN_SetSpawnGroup()
   &FN_InitializeRibbon_v4(
    Custom Offset=true
    Enable Skeletal Position=true
    Facing Axis:ENiagaraOrientationAxis=::YAxis
    Lifetime=1
    Local Vector Offset:Vector3f=&Multiply_VectorByFloat(
     Vector=&NormalizeVector(
      Vector To Normalize=&TransformVector(Vector=$System.SkydiveVelocity, Source Space=::World, Destination Space=::Local)
      Safe Normalize=true
      Safe Fallback Threshold=0
      Safe Fallback Vector=<0 0 0.001>
     )
     Float=&Lerp_Float(
      A=15
      B=15
      Alpha=&MakeFloatFromInt(Int=&Divide_Int(A:NiagaraInt32=&SpawnGroup(), B:NiagaraInt32=2))
     )
    )
    Offset Amount:NiagaraFloat=&Lerp_Float(A=1, B=2.5, Alpha=$System.Dive)
    Ribbon Width:NiagaraFloat=&Lerp_Float(
     A=25
     B=20
     Alpha=&RemapRange(
      Input Value To Remap=&VectorLength(Vector=$System.SkydiveVelocity)
      Input Min=1200
      Input Max=4000
      Output Min=0
      Output Max=1
     )
    )
    SkeletalMesh:NiagaraDataInterfaceSkeletalMesh=$User.SkeletalMesh
   )
   &FN_FlipRibbonV() {enabled=false}
   &AddVelocity(Velocity=&Multiply_VectorByFloat(
     Vector=&TransformVector(Vector=$System.SkydiveVelocity, Source Space=::World, Destination Space=::Local, Apply Scale=true)
     Float=-0.8
    ))
   /SetVariables(Particles.Lifetime=&Sine(
     Normalized Angle=&Add_Float(
      A=$Emitter.Age
      B=&SelectFloatFromArray(
       Direct Array Index=&SpawnGroup()
       Float Selection Array=DataInterface() {propertyValues {FloatData [0,0.25,0.5,0.75]}}
      )
     )
     Period=&SelectFloatFromArray(
      Direct Array Index=&SpawnGroup()
      Float Selection Array=DataInterface() {propertyValues {FloatData [1.8,1,1.5,2]}}
     )
     Scale=&Lerp_Float(
      A=0.02
      B=0
      Alpha=&RemapRange(
       Input Value To Remap=&VectorLength(Vector=$System.SkydiveVelocity)
       Input Min=1200
       Input Max=2000
       Output Min=0
       Output Max=1
      )
     )
     Bias=&Multiply_Float(
      A=&Lerp_Float(
       A=0.2
       B=0.1
       Alpha=&RemapRange(
        Input Value To Remap=&VectorLength(Vector=$System.SkydiveVelocity)
        Input Min=1200
        Input Max=2000
        Output Min=0
        Output Max=1
       )
      )
      B=&Lerp_Float(
       A=1
       B=0.4
       Alpha=&MakeFloatFromInt(Int=&Divide_Int(A:NiagaraInt32=&SpawnGroup(), B:NiagaraInt32=2))
      )
     )
    )) {isSetParameters=true, module=SetVariables_#9}
   /SetVariables(Particles.RibbonWidth=&Multiply_Float(
     A=$Particles.RibbonWidth
     B=&Lerp_Float(
      A=0.9
      B=0.7
      Alpha=&MakeFloatFromInt(Int=&Divide_Int(A:NiagaraInt32=&SpawnGroup(), B:NiagaraInt32=2))
     )
    )) {isSetParameters=true}
   &SkeletalMeshLocation(
    Bone / Socket Index=$Particles.SpawnGroup
    Bone Sampling Mode=::NewEnumerator4
    Interpolate Mesh Sampling=false
    MaskSpawnGroup=false
    Orientation Sampling=::NewEnumerator1
    Position Sampling=::NewEnumerator1
    Skeletal Mesh=$User.SkeletalMesh
   ) {enabled=false}
   /SetVariables(Particles.OffsetVector:Vector3f=&Add_Vector(
     A=$Particles.OffsetVector
     B=&RotateVectorByQuaternion(
      Vector To Rotate=&SelectVectorFromArray(
       Direct Array Index=&SpawnGroup()
       Vector Selection Array=DataInterface() {propertyValues {FloatData [<5 5 0>, <-5 -5 0>, <2 -10 0>, <-2 10 0>]}}
      )
      Quaternion=$Particles.SampleSkeletalMesh.SampledRotation
     )
    )) {enabled=false, isSetParameters=true}
   /SetVariables(Particles.Position=&AddVectorToPosition(
     Position=$Particles.Position
     Vector=&RotateVectorByQuaternion(
      Vector To Rotate=&SelectVectorFromArray(
       Direct Array Index=&SpawnGroup()
       Vector Selection Array=DataInterface() {propertyValues {FloatData [<5 5 0>, <-5 -5 0>, <2 -10 0>, <-2 10 0>]}}
      )
      Quaternion=$Particles.SampleSkeletalMesh.SampledRotation
     )
    )) {enabled=false, isSetParameters=true}
   /SetVariables(Particles.Position=&AddVectorToPosition(
     Position=$Particles.Position
     Vector=&RotateVectorByQuaternion(
      Vector To Rotate=&SelectVectorFromArray(
       Direct Array Index=&SpawnGroup()
       Vector Selection Array=DataInterface() {propertyValues {FloatData [<0 0 0>, <0 0 0>, <0 -5 0>, <0 5 0>]}}
      )
      Quaternion=$Particles.SkeletalMeshLocation.SampledRotation
     )
    )) {enabled=false, isSetParameters=true}
  ]
  particleUpdateScript [
   &InterpolateOverTime(
    Rate Of Change:NiagaraFloat=&FloatFromCurve(
     FloatCurve=DataInterface() {propertyValues {Curve {keys [
         {arriveTangent=-1.9, interpMode=RCIM_Cubic, leaveTangent=-1.9, tangentMode=RCTM_User, time=0, value=1}
         {interpMode=RCIM_Cubic, tangentMode=RCTM_User, time=1, value=0}
        ]}}}
     Scale Curve=20
    )
    Target Vector Value:Vector3f=$System.SkydiveVelocity
    Variable Type=::NewEnumerator2
   )
   /SetVariables(Particles.Velocity=&Multiply_VectorByFloat(
     Vector=&NormalizeVector(
      Vector To Normalize=&TransformVector(
       Vector=$Particles.InterpolateOverTime.InterpolateOverTime_V3_.Moving Average
       Source Space=::World
       Destination Space=::Local
       Apply Scale=true
      )
      Safe Normalize=true
      Safe Fallback Threshold=0
      Safe Fallback Vector=<0 0 0.001>
     )
     Float=&Multiply_Float(A=-1, B=&VectorLength(Vector=$Particles.Velocity))
    )) {isSetParameters=true}
   &SolveForcesAndVelocity(Clamp Velocity=false, Limit Acceleration=false)
   &DynamicMaterialParameters(
    Index 0 Param 1=&self:NormalizedRibbonID(Number of Ribbons:NiagaraInt32=4, Ribbon ID:NiagaraID=$Particles.RibbonID)
    Index 0 Param 2=$System.Dive
    Index 0 Param 3=&Lerp_Float(
     A=-15
     B=-15
     Alpha=&MakeFloatFromInt(Int=&Divide_Int(A:NiagaraInt32=&SpawnGroup(), B:NiagaraInt32=2))
    )
    Index 1 Param 1=&MakeFloatFromVector(Vector=$System.SkydiveVelocity, Channel=::NewEnumerator0)
    Index 1 Param 2=&MakeFloatFromVector(Vector=$System.SkydiveVelocity, Channel=::NewEnumerator1)
    Index 1 Param 3=&MakeFloatFromVector(Vector=$System.SkydiveVelocity, Channel=::NewEnumerator2)
    Module.Param0WriteEnabled=true
    Module.Param0WriteEnabled1=true
    Module.Param1WriteEnabled=true
    Module.Param1WriteEnabled1=true
    Module.Param2WriteEnabled=true
    Module.Param2WriteEnabled1=true
    Module.Param3WriteEnabled=false
    Module.Param3WriteEnabled1=false
    Write Parameter Index 1=true
   )
   /SetVariables(Particles.PositionOffset:Vector3f=&Multiply_VectorByFloat(
     Vector=&NormalizeVector(
      Vector To Normalize=&TransformVector(Vector=$System.SkydiveVelocity, Source Space=::World, Destination Space=::Local)
      Safe Normalize=true
      Safe Fallback Threshold=0
      Safe Fallback Vector=<0 0 0.001>
     )
     Float=300
    )) {enabled=false, isSetParameters=true}
   &ScaleRibbonWidth(Ribbon Width Scale=&FloatFromCurve(FloatCurve=DataInterface() {propertyValues {Curve {keys [{interpMode=RCIM_Cubic, tangentMode=RCTM_User, time=0, value=1.4}, {interpMode=RCIM_Cubic, tangentMode=RCTM_User, time=1, value=0}]}}}))
   &FN_UpdateRibbon_v4(
    Detach Locked Particle=true
    Enable Skeletal Position=true
    Locked Particle Detach Time=0
    SkeletalMesh:NiagaraDataInterfaceSkeletalMesh=$User.SkeletalMesh
   )
   &FN_RibbonMaterial() {enabled=false}
  ]
  rendererClasses [&NiagaraRibbonRendererProperties]
  renderers [{rendererClass=&NiagaraRibbonRendererProperties}]
 }
 Light_and_Glow {
  emitterSpawnScript [&FN_PreserveSkydiveUserVariables() {module=PreserveUserVariables}]
  emitterUpdateScript [
   &EmitterState(Inactive Response=::NewEnumerator1, Life Cycle Mode=::NewEnumerator1, Loop Duration=0, UseLoopDelay=false)
   &SpawnBurst_Instantaneous(Spawn Count=4, Spawn Probability=HlslExpression() {hlslExpression="Emitter.LoopCount == 0"}, Use Spawn Probability=true) {module=SpawnBurst_Instantaneous001}
  ]
  particleSpawnScript [
   &FN_SetSpawnGroup()
   &InitializeParticle(
    Color Mode=::NewEnumerator0
    Lifetime=0
    Sprite Size Mode=::NewEnumerator3
    Uniform Sprite Size=25
    UsePositionOffset=false
    Write Material Random=true
   ) {module=InitializeParticle001}
   &Color(Color=$NPC.FortniteNPC.FortniteActiveTimeOfDay)
  ]
  particleUpdateScript [
   &ParticleState(Kill Particles When Lifetime Has Elapsed=false)
   &SkeletalMeshLocation(
    Bone / Socket Index=$Particles.SpawnGroup
    Bone Sampling Mode=::NewEnumerator4
    Inherit Velocity=false
    Interpolate Mesh Sampling=false
    MaskSpawnGroup=false
    Sampled Position Offset=&SelectVectorFromArray(
     Direct Array Index=&SpawnGroup()
     Vector Selection Array=DataInterface() {propertyValues {FloatData [<0 0 0>, <0 1 1>, <5 -7 0>, <-5 7 0>]}}
    )
    Sampled Position Offset Coordinate Space:ENiagaraMeshTransforms=::NewEnumerator3
    Skeletal Mesh=$User.SkeletalMesh
   )
   /SetVariables(Particles.LightColor=&MultiplyLinearColor(
     Linear Color=<rgba 0.1667 0 5 1>
     Scale Factor=&MakeVector4_FromFloat(Float=&FloatFromCurve(FloatCurve=DataInterface() {propertyValues {Curve {keys [<at 0 1>,<at 1 2>]}}}, CurveIndex=$System.Dive))
    )) {isSetParameters=true, module=SetVariables_#7}
   /SetVariables(Particles.SpriteSize=&Multiply_Vector2D_ByFloat(Float=&Lerp_Float(A=1, B=2, Alpha=$System.Dive), Vector2D=$Particles.Initial.SpriteSize)) {isSetParameters=true, module=SetVariables_#8}
  ]
  rendererClasses [&NiagaraLightRendererProperties, &NiagaraSpriteRendererProperties]
  renderers [
   {rendererClass=&NiagaraLightRendererProperties, rendererIndex=0}
   {rendererClass=&NiagaraSpriteRendererProperties, rendererIndex=1}
  ]
 }
}
summary {
 systemName=NS_Contrail_Chainmail_A
 emitters [
  {emitterName=Chains, rendererClasses [&NiagaraMeshRendererProperties]}
  {emitterName=Cones, rendererClasses [&NiagaraMeshRendererProperties]}
  {emitterName=Flames_Frontend, rendererClasses [&NiagaraRibbonRendererProperties]}
  {emitterName=FlamesSprites, rendererClasses [&NiagaraSpriteRendererProperties]}
  {emitterName=Embers, rendererClasses [&NiagaraSpriteRendererProperties]}
  {emitterName=Flames_Gameplay_LocalSpace, rendererClasses [&NiagaraRibbonRendererProperties]}
  {
   emitterName=Light_and_Glow
   rendererClasses [&NiagaraLightRendererProperties, &NiagaraSpriteRendererProperties]
  }
 ]
 userVariables [
  User.bIsFrontEnd {type=NiagaraBool}
  User.bIsFrontEndPreview {type=NiagaraBool}
  User.ColorChange {type=LinearColor, description="", defaultValue {struct=LinearColor, value=<rgba 1 1 1 1>}}
  User.ForwardDot {type=NiagaraFloat, defaultValue=0, description=""}
  User.ParaGlideLeanAlpha {type=NiagaraFloat, defaultValue=0, description=""}
  User.PlayerSpeed {type=NiagaraFloat, defaultValue=0, description=""}
  User.RibbonIDOffset {type=NiagaraInt32}
  User.SkeletalMesh {
   type=NiagaraDataInterfaceSkeletalMesh
   defaultValue {
    struct=NiagaraExt_VariableValue_DataInterface
    value {dataInterface=&self:NiagaraDataInterfaceSkeletalMesh_2}
   }
  }
  User.SkeletalMeshFrontend {
   type=NiagaraDataInterfaceSkeletalMesh
   defaultValue {
    struct=NiagaraExt_VariableValue_DataInterface
    value {dataInterface=&self:NiagaraDataInterfaceSkeletalMesh_0}
   }
  }
 ]
}
